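# Niagara water splash effect for the QuicheLorraineLime emote, built from one-shot-style bursts of droplets, spray and splashes.
# Drip_Drop throws falling droplet sprites, while Edge_Spray_Mesh002 adds scaling, fading spray meshes driven by dynamic material parameters.
# Splashes_compressed packs several splash layers into one emitter, switching behavior by spawn group across five sprite renderers; its bursts use groups 1, 3 and 5.
# Time-of-day inputs adjust color and alpha; Droplets001 and LWE_RIPPLE are empty.
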NiagaraSystem /BRCosmetics/Animation/Game/MainPlayer/Emotes/QuicheLorraineLime/FX/NS_Emote_QuicheLorraineLime_Splash
dependencies {
 usedDataInterfaces [
  &NiagaraDataInterfaceVector2DCurve
  &NiagaraDataInterfaceCurve
  &NiagaraDataInterfaceVectorCurve
  &NiagaraDataInterfaceSpriteRendererInfo
 ]
 usedDynamicInputs [&RandomRangeLinearColor]
 usedModules [
  &self:SystemSpawnScript.SystemScriptSource.SystemScriptGraph.NiagaraNodeAssignment_3.SetVariables_#1
  &SystemState
 ]
}
emitters {
 Drip_Drop {
  emitterUpdateScript [
   &EmitterState(Life Cycle Mode=::NewEnumerator1, Loop Behavior=::NewEnumerator1, Loop Delay=0.1, Loop Duration=1, UseLoopDelay=true)
   &SpawnBurst_Instantaneous(Spawn Count=15, Spawn Time=0.1, Use Spawn Probability=false, UseLoopCountLimit=false) {module=SpawnBurst_Instantaneous001}
   &SpawnBurst_Instantaneous(Spawn Count=20, Use Spawn Probability=false, UseLoopCountLimit=false)
  ]
  particleSpawnScript [
   &InitializeParticle(
    Color=<rgba 0.7878 0.9763 1 0.05>
    Lifetime Max=0.6
    Lifetime Min=0.2
    Lifetime Mode=::NewEnumerator1
    Position Offset=<0 0 6.4957>
    Sprite Size Max=<2 8>
    Sprite Size Min=<1.5 5>
    Sprite Size Mode=::NewEnumerator2
    UsePositionOffset=true
    Write Material Random=true
   ) {module=InitializeParticle001}
   &AddVelocity(
    Cone Angle=&RandomRangeFloat(Minimum=90, Maximum=0)
    Cone Axis=<0 0 5>
    Distribution Along Cone Axis=0.25
    Inner Cone Angle=90
    Module.UseConeFalloff=false
    Module.UseVelDistribution=false
    Override Seed=false
    Speed Falloff From Cone Axis=0.5
    Velocity Mode=::NewEnumerator2
    Velocity Speed=&RandomRangeFloat(Minimum=100, Maximum=300)
   ) {module=AddVelocity001}
  ]
  particleUpdateScript [
   &CameraOffset(Camera Offset Amount=&RandomRangeFloat(Minimum=0, Maximum=1))
   &ParticleState()
   &GravityForce(Gravity=<0 0 -600>)
   &SolveForcesAndVelocity(Clamp Velocity=false, Limit Acceleration=false)
   &ScaleSpriteSize(Scale Factor=&Vector2DFromCurve(
     Vector2Curve=DataInterface() {propertyValues {
       XCurve {keys [
         {arriveTangent=5.9728, interpMode=RCIM_Cubic, leaveTangent=5.9728, tangentMode=RCTM_User, time=0, value=0.1}
         {arriveTangent=-0.1, interpMode=RCIM_Cubic, leaveTangent=-0.1, tangentMode=RCTM_User, time=0.18, value=0.99}
         {interpMode=RCIM_Cubic, tangentMode=RCTM_User, time=1, value=0}
        ]}
       YCurve {keys [
         {arriveTangent=6.2715, interpMode=RCIM_Cubic, leaveTangent=6.2715, tangentMode=RCTM_User, time=0, value=0.1}
         {arriveTangent=-0.1001, interpMode=RCIM_Cubic, leaveTangent=-0.1001, tangentMode=RCTM_User, time=0.18, value=0.99}
         {interpMode=RCIM_Cubic, tangentMode=RCTM_User, time=0.9989, value=0}
        ]}
      }}
     ScaleCurve=<2 2>
    ))
   &ScaleColor(
    Scale Alpha=&FloatFromCurve(FloatCurve=DataInterface() {propertyValues {Curve {
        defaultValue=1
        keys [{interpMode=RCIM_Cubic, tangentMode=RCTM_User, time=0, value=1}, {interpMode=RCIM_Cubic, tangentMode=RCTM_User, time=1, value=0}]
       }}})
    Scale RGB=&MakeVectorFromLinearColorRGB(Color=&FNTimeOfDayColor(
      Night Color=<rgba 0.218 0.3666 0.1807 0.1>
      Morning Color=<rgba 1 1 1 1>
      Day Color=<rgba 1 1 1 1>
      Evening Color=<rgba 0.218 0.3666 0.1807 0.1>
     ))
    ScaleA=true
    ScaleRGB=true
   )
  ]
  rendererClasses [&NiagaraSpriteRendererProperties]
  renderers [{rendererClass=&NiagaraSpriteRendererProperties}]
 }
 Droplets001 {rendererClasses [], renderers []}
 Edge_Spray_Mesh002 {
  emitterUpdateScript [
   &EmitterState(Life Cycle Mode=::NewEnumerator1, Loop Behavior=::NewEnumerator1, UseLoopDelay=false)
   &SpawnRate(SpawnRate=500, Use Spawn Probability=false) {enabled=false}
   &SpawnBurst_Instantaneous(Spawn Count=25, Use Spawn Probability=false, UseLoopCountLimit=false)
  ]
  particleSpawnScript [
   &InitializeParticle(
    Lifetime Max=0.75
    Lifetime Min=0.5
    Lifetime Mode=::NewEnumerator1
    Mesh Scale Max=<0.6 0.6 0.6>
    Mesh Scale Min=<0.4 0.4 0.4>
    Mesh Scale Mode=::NewEnumerator2
    UsePositionOffset=false
    Write Material Random=true
   ) {module=InitializeParticle001}
   &ShapeLocation(
    MaskSpawnGroup=false
    Offset=<0 0 -5>
    Offset Mode=::NewEnumerator0
    Rotation Axis=<0 0 1>
    Rotation Mode=::NewEnumerator1
    Scale Mode:ENiagara_ScaleMode=::NewEnumerator1
    Sphere Distribution=::NewEnumerator3
    Sphere Radius=1
   )
   /SetVariables(Particles.FrontToBack:NiagaraFloat=&RandomRangeFloat(Minimum=0.5, Maximum=1.2), Particles.LeanLerp=1) {isSetParameters=true}
   /SetVariables(Particles.Scale=&Multiply_Vector(
     A=&VectorFromCurve(
      VectorCurve=DataInterface() {propertyValues {
        XCurve {keys [<at 0 1 RCIM_Cubic>, {arriveTangent=1.7846, interpMode=RCIM_Cubic, leaveTangent=1.7846, tangentMode=RCTM_User, time=0, value=0.4}]}
        YCurve {keys [<at 0 1 RCIM_Cubic>, {arriveTangent=1.7846, interpMode=RCIM_Cubic, leaveTangent=1.7846, tangentMode=RCTM_User, time=0, value=0.4}]}
        ZCurve {keys [{arriveTangent=1.7846, interpMode=RCIM_Cubic, leaveTangent=1.7846, tangentMode=RCTM_User, time=0, value=1}]}
       }}
      CurveIndex=$Particles.FrontToBack
      Scale Curve=&Multiply_VectorByFloat(Vector=$Particles.Scale, Float=&Lerp_Float(A=0.8, B=1, Alpha=1))
     )
     B=&Lerp_Vector(A=<0.6 0.6 0.6>, B=<0.6 0.6 0.6>, Alpha=$Particles.LeanLerp)
    )) {isSetParameters=true}
   &AddVelocity(Velocity=&RandomRangeVector(Minimum=<-50 -50 10>, Maximum=<50 50 30>)) {module=AddVelocity001}
   &CameraOffset(Camera Offset Amount=&RandomRangeFloat(Minimum=0, Maximum=2))
  ]
  particleUpdateScript [
   &ParticleState()
   /SetVariables(Particles.Scale=&Multiply_Vector(
     A=$Particles.Initial.Scale
     B=&VectorFromCurve(
      VectorCurve=DataInterface() {propertyValues {
        XCurve {keys [
          {arriveTangent=4.6994, interpMode=RCIM_Cubic, leaveTangent=4.6994, tangentMode=RCTM_User, time=0, value=0.0795}
          {arriveTangent=0.4589, interpMode=RCIM_Cubic, leaveTangent=0.4589, tangentMode=RCTM_User, time=0.05, value=0.7609}
          <at 1 1>
         ]}
        YCurve {keys [{arriveTangent=2.5596, interpMode=RCIM_Cubic, leaveTangent=2.5596, tangentMode=RCTM_User, time=0, value=0.0767}, <at 1 1>]}
        ZCurve {keys [
          <at 0 0 RCIM_Cubic>
          {arriveTangent=-2.3619, interpMode=RCIM_Cubic, leaveTangent=-2.3619, tangentMode=RCTM_User, time=0.05, value=0.9567}
          <at 1 0>
         ]}
       }}
      Scale Curve=<0.4 0.9 0.9>
     )
    )) {isSetParameters=true}
   &SolveForcesAndVelocity(Clamp Velocity=false, Limit Acceleration=false)
   &ScaleColor(
    Scale Alpha=&FloatFromCurve(
     FloatCurve=DataInterface() {propertyValues {Curve {defaultValue=0, keys [{tangentMode=RCTM_User, time=0, value=1}, {tangentMode=RCTM_User, time=1, value=0}]}}}
     Scale Curve=&FNTimeOfDayScalar(Night Scale=1, Morning Scale=1, Evening Scale=1)
    )
    Scale RGB=&MakeVectorFromLinearColorRGB(Color=&FNTimeOfDayColor(
      Night Color=<rgba 0.2723 0.2465 0.1609 0.35>
      Morning Color=<rgba 0.7935 0.849 0.6829 0.5>
      Day Color=<rgba 0.7935 0.849 0.6829 0.5>
      Evening Color=<rgba 0.2723 0.2465 0.1609 0.35>
     ))
    ScaleA=true
    ScaleRGB=true
   )
   &DynamicMaterialParameters(
    Index 0 Param 1=&FloatFromCurve(FloatCurve=DataInterface() {propertyValues {Curve {keys [
         {arriveTangent=1.2629, interpMode=RCIM_Cubic, leaveTangent=1.2629, tangentMode=RCTM_User, time=-0.0045, value=-1.9799}
         {arriveTangent=4.4591, interpMode=RCIM_Cubic, leaveTangent=4.4591, tangentMode=RCTM_User, time=1, value=0}
        ]}}})
    Index 0 Param 2=&FloatFromCurve(FloatCurve=DataInterface() {propertyValues {Curve {keys [<at 0.0102 4>,<at 0.9749 0>]}}})
    Index 0 Param 3=&MakeFloatFromVector(Vector=$Particles.Scale, Channel=::NewEnumerator0)
    Index 0 Param 4=1
    Index 1 Param 1=&FloatFromCurve(FloatCurve=DataInterface() {propertyValues {Curve {keys [<at 0.7 0 RCIM_Cubic>, {arriveTangent=6.4641, interpMode=RCIM_Cubic, leaveTangent=6.4641, tangentMode=RCTM_User, time=1, value=1}]}}})
    Module.Index0Float0Write=true
    Module.Index0Float1Write=false
    Module.Index0Float2Write=true
    Module.Index0Float3Write=true
    Module.Index1Float0Write=true
    Module.Index1Float1Write=true
    Module.Index1Float2Write=true
    Module.Index1Float3Write=true
    Write Parameter Index 1=::NewEnumerator1
   )
  ]
  rendererClasses [&NiagaraMeshRendererProperties]
  renderers [{rendererClass=&NiagaraMeshRendererProperties, rendererIndex=0}]
 }
 LWE_RIPPLE {rendererClasses [], renderers []}
 Splashes_compressed {
  emitterUpdateScript [
   &EmitterState(Life Cycle Mode=::NewEnumerator1, Loop Behavior=::NewEnumerator1, Loop Duration=3, UseLoopDelay=false)
   &SpawnBurst_Instantaneous(Spawn Count=15, Spawn Group=5, Spawn Time=0.15, Use Spawn Probability=false, UseLoopCountLimit=false) {module=SpawnBurst_Instantaneous005}
   &SpawnBurst_Instantaneous(Spawn Count=15, Spawn Group=3, Spawn Time=0.25, Use Spawn Probability=false, UseLoopCountLimit=false) {module=SpawnBurst_Instantaneous003}
   &SpawnBurst_Instantaneous(Spawn Count=8, Spawn Group=1, Use Spawn Probability=false) {module=SpawnBurst_Instantaneous001}
  ]
  particleSpawnScript [
   /SetVariables(Particles.VisibilityTag=&SpawnGroup()) {isSetParameters=true}
   /SetVariables(Particles.Scaler:NiagaraFloat=&FloatFromCurve(
     FloatCurve=DataInterface() {propertyValues {Curve {keys [
         <at 0 0 RCIM_Cubic>
         {arriveTangent=2.9389, interpMode=RCIM_Cubic, leaveTangent=2.9389, tangentMode=RCTM_User, time=0.5, value=0.5}
         <at 1 1 RCIM_Cubic>
        ]}}}
     CurveIndex=&RandomRangeFloat(Minimum=0, Maximum=1)
    )) {isSetParameters=true}
   &InitializeParticle(
    Color=&MergeHelper_ColorSwitch(
     0=$System.Water Splashes color
     1=<rgba 1 1 1 1>
     2=$System.Water Splashes color
     3=$System.Water Splashes color
     4=$System.Water Splashes color
     5=$System.Water Splashes color
     Index=&SpawnGroup()
     Num=::NewEnumerator4
    )
    Lifetime=&MergeHelper_FloatSwitch(
     0=&RandomRangeFloat(Minimum=0.3, Maximum=0.65)
     1=&RandomRangeFloat(Minimum=0.3, Maximum=0.6)
     2=&RandomRangeFloat(Minimum=0.2, Maximum=0.55)
     3=&RandomRangeFloat(Minimum=0.3, Maximum=0.85)
     4=&RandomRangeFloat(Minimum=0.45, Maximum=0.65)
     5=&RandomRangeFloat(Minimum=0.65, Maximum=0.85)
     Index=&SpawnGroup()
     Num=::NewEnumerator4
    )
    Sprite Rotation Angle=&MergeHelper_FloatSwitch(
     0=&RandomRangeFloat(Minimum=0, Maximum=1)
     1=&RandomRangeFloat(Minimum=0, Maximum=1)
     2=&RandomRangeFloat(Minimum=0, Maximum=0.5)
     3=&RandomRangeFloat(Minimum=0, Maximum=0.2)
     4=&RandomRangeFloat(Minimum=0, Maximum=1)
     5=&RandomRangeFloat(Minimum=0, Maximum=0.2)
     Index=&SpawnGroup()
     Num=::NewEnumerator4
    )
    Sprite Rotation Mode=::NewEnumerator1
    Sprite Size Mode=::NewEnumerator3
    Uniform Sprite Size=&MergeHelper_FloatSwitch(
     0=&RandomRangeFloat(Minimum=5, Maximum=10)
     1=&MakeFloatFromVector2d(Vector 2D=&RandomRangeVector2D(Minimum=<2 4>, Maximum=<3 6>))
     2=&MakeFloatFromVector2d(Vector 2D=&Multiply_Vector2D_ByFloat(Float=10, Vector2D=&RandomRangeVector2D(Minimum=<1.5 5>, Maximum=<2 10>)))
     3=&RandomRangeFloat(Minimum=10, Maximum=20)
     4=&RandomRangeFloat(Minimum=15, Maximum=50)
     5=&RandomRangeFloat(Minimum=10, Maximum=20)
     Index=&SpawnGroup()
     Num=::NewEnumerator4
    )
    UsePositionOffset=true
    Write Material Random=true
   )
   &AddVelocity(Velocity=&MergeHelper_VectorSwitch(
     0=&RandomRangeVector(Minimum=<100 100 250>, Maximum=<-100 -100 750>)
     1=&RandomRangeVector(Minimum=<0 0 10>, Maximum=<0 0 30>)
     2=&RandomRangeVector(Minimum=<-5 -5 10>, Maximum=<5 5 20>)
     3=&RandomRangeVector(Minimum=<-75 -75 75>, Maximum=<75 75 150>)
     4=&RandomRangeVector(Minimum=<-50 -50 75>, Maximum=<50 50 250>)
     5=&RandomRangeVector(Minimum=<-50 -50 200>, Maximum=<50 50 300>)
     Index=&SpawnGroup()
     Num=::NewEnumerator4
    ))
   &SkeletalMeshLocation(
    MaskSpawnGroup=false
    Mesh Sampling Type=::NewEnumerator5
    Orientation Sampling=::NewEnumerator1
    Sampled Position Offset=&MergeHelper_VectorSwitch(0=<-8 5 0>, 1=<-5 0 0>, 2=<-5 0 15>, 3=<0 0 5>, 4=<0 0 0>, Index=&SpawnGroup(), Num=::NewEnumerator3)
    Skeletal Mesh=$System.Skeletal Mesh
    Socket Sampling Mode=::NewEnumerator5
   ) {enabled=false}
   &ShapeLocation(
    MaskSpawnGroup=false
    Non Uniform Scale=&MergeHelper_VectorSwitch(
     0=<2 2 1>
     1=<2 2 0.1>
     2=<0.5 0.5 0.1>
     3=<2 2 1>
     4=<1 1 1>
     5=<1 1 1>
     Index=&SpawnGroup()
     Num=::NewEnumerator4
    )
    Offset=&MergeHelper_VectorSwitch(
     0=<0 0 -25>
     1=<0 0 -25>
     2=<0 0 0>
     3=<0 0 -25>
     4=<0 0 0>
     5=<0 0 -25>
     Index=&SpawnGroup()
     Num=::NewEnumerator4
    )
    Offset Mode=::NewEnumerator0
    Override Seed=false
    Sphere Radius=&MergeHelper_FloatSwitch(0=8, 1=5, 2=3, 3=5, 4=1, 5=5, Index=&SpawnGroup(), Num=::NewEnumerator4)
    Sphere Surface Distribution=0.4357
    UseOwnerScale=true
   )
   &SubUVAnimation(
    End Frame Range Override=3
    Evaluation Type=::NewEnumerator0
    "Sprite Renderer "=DataInterface() {propertyValues {SpriteRenderer=&self:Splashes_compressed_0.NiagaraSpriteRendererProperties_0}}
    SubUV Animation Mode:ENiagara_SubUVLookupModeV2=::NewEnumerator2
    UseEndFrame=true
    UseStartFrame=true
   )
   &SubUVAnimation(
    End Frame Range Override=4
    Evaluation Type=::NewEnumerator0
    "Sprite Renderer "=DataInterface() {propertyValues {SpriteRenderer=&self:Splashes_compressed_0.NiagaraSpriteRendererProperties_2}}
    SubUV Animation Mode:ENiagara_SubUVLookupModeV2=::NewEnumerator2
    UseEndFrame=true
    UseStartFrame=true
   ) {module=SubUVAnimation001}
   &SubUVAnimation(
    Direct Frame Index:NiagaraFloat=&RandomRangeFloat(Minimum=0, Maximum=10)
    "Sprite Renderer "=DataInterface() {propertyValues {SpriteRenderer=&self:Splashes_compressed_0.NiagaraSpriteRendererProperties_6}}
    SubUV Animation Mode:ENiagara_SubUVLookupModeV2=::NewEnumerator5
   ) {module=SubUVAnimation002}
   &SubUVAnimation(
    End Frame Range Override=3
    Evaluation Type=::NewEnumerator0
    "Sprite Renderer "=DataInterface() {propertyValues {SpriteRenderer=&self:Splashes_compressed_0.NiagaraSpriteRendererProperties_7}}
    SubUV Animation Mode:ENiagara_SubUVLookupModeV2=::NewEnumerator2
    UseEndFrame=true
    UseStartFrame=true
   ) {module=SubUVAnimation003}
   &SubUVAnimation(
    End Frame Range Override=3
    Evaluation Type=::NewEnumerator0
    "Sprite Renderer "=DataInterface() {propertyValues {SpriteRenderer=&self:Splashes_compressed_0.NiagaraSpriteRendererProperties_9}}
    SubUV Animation Mode:ENiagara_SubUVLookupModeV2=::NewEnumerator2
    UseEndFrame=true
    UseStartFrame=true
   ) {module=SubUVAnimation004}
   &CameraOffset(Camera Offset Amount=&MergeHelper_FloatSwitch(
     0=&RandomRangeFloat(Minimum=0, Maximum=10)
     1=&RandomRangeFloat(Minimum=6, Maximum=15)
     2=&RandomRangeFloat(Minimum=6, Maximum=20)
     3=&RandomRangeFloat(Minimum=0, Maximum=3)
     4=0
     5=&RandomRangeFloat(Minimum=0, Maximum=3)
     Index=&SpawnGroup()
     Num=::NewEnumerator4
    ))
   &ApplyInitialForces()
   /SetVariables(Particles.SubImageIndex 1:NiagaraFloat=&RandomRangeFloat(Minimum=0, Maximum=4)) {isSetParameters=true}
  ]
  particleUpdateScript [
   &ParticleState()
   &ScaleSpriteSize(
    Scale Sprite Size Mode=::NewEnumerator0
    Uniform Scale Factor=&MergeHelper_FloatSwitch(
     0=&FloatFromCurve(FloatCurve=DataInterface() {propertyValues {Curve {keys [
          {arriveTangent=10.0275, interpMode=RCIM_Cubic, leaveTangent=10.0275, tangentMode=RCTM_User, time=0, value=0}
          {arriveTangent=1.6841, interpMode=RCIM_Cubic, leaveTangent=1.6841, tangentMode=RCTM_User, time=1, value=5}
         ]}}})
     1=&FloatFromCurve(FloatCurve=DataInterface() {propertyValues {Curve {keys [
          {arriveTangent=4.1785, interpMode=RCIM_Cubic, leaveTangent=4.1785, tangentMode=RCTM_User, time=0, value=0.15}
          {arriveTangent=1.1179, interpMode=RCIM_Cubic, leaveTangent=1.1179, tangentMode=RCTM_User, time=0.2042, value=0.6228}
          {arriveTangent=0.0224, interpMode=RCIM_Cubic, leaveTangent=0.0224, tangentMode=RCTM_User, time=1, value=1}
         ]}}})
     2=&FloatFromCurve(FloatCurve=DataInterface() {propertyValues {Curve {keys [
          {arriveTangent=0.1295, interpMode=RCIM_Cubic, leaveTangent=0.1295, tangentMode=RCTM_User, time=0, value=0}
          {arriveTangent=1, interpMode=RCIM_Cubic, leaveTangent=1, time=0.35, value=0.9}
          <at 1 1 RCIM_Cubic>
         ]}}})
     3=&FloatFromCurve(FloatCurve=DataInterface() {propertyValues {Curve {keys [
          {arriveTangent=4.1785, interpMode=RCIM_Cubic, leaveTangent=4.1785, tangentMode=RCTM_User, time=0, value=0.15}
          {arriveTangent=1.1179, interpMode=RCIM_Cubic, leaveTangent=1.1179, tangentMode=RCTM_User, time=0.2042, value=0.6228}
          {arriveTangent=0.0224, interpMode=RCIM_Cubic, leaveTangent=0.0224, tangentMode=RCTM_User, time=1, value=1}
         ]}}})
     4=&Lerp_Float(
      A=&FloatFromCurve(FloatCurve=DataInterface() {propertyValues {Curve {keys [
           {arriveTangent=2.9966, interpMode=RCIM_Cubic, leaveTangent=2.9966, tangentMode=RCTM_User, time=0, value=0}
           {arriveTangent=1, interpMode=RCIM_Cubic, leaveTangent=1, time=0.3, value=0.6}
           {arriveTangent=0.5346, interpMode=RCIM_Cubic, leaveTangent=0.5346, tangentMode=RCTM_User, time=1, value=1}
          ]}}})
      B=&FloatFromCurve(FloatCurve=DataInterface() {propertyValues {Curve {keys [
           {arriveTangent=1.5909, interpMode=RCIM_Cubic, leaveTangent=1.5909, tangentMode=RCTM_User, time=0, value=0}
           {arriveTangent=0.5346, interpMode=RCIM_Cubic, leaveTangent=0.5346, tangentMode=RCTM_User, time=1, value=1}
          ]}}})
      Alpha=$Particles.Scaler
     )
     5=&FloatFromCurve(FloatCurve=DataInterface() {propertyValues {Curve {keys [
          {arriveTangent=4.1785, interpMode=RCIM_Cubic, leaveTangent=4.1785, tangentMode=RCTM_User, time=0, value=0}
          {arriveTangent=1.1179, interpMode=RCIM_Cubic, leaveTangent=1.1179, tangentMode=RCTM_User, time=0.2042, value=0.6228}
          {arriveTangent=0.0224, interpMode=RCIM_Cubic, leaveTangent=0.0224, tangentMode=RCTM_User, time=1.0045, value=0.824}
         ]}}})
     Index=&SpawnGroup()
     Num=::NewEnumerator4
    )
   )
   &SpriteRotationRate(Rotation Rate=&MergeHelper_FloatSwitch(
     0=&Lerp_Float(A=-150, B=150, Alpha=$Particles.Initial.MaterialRandom)
     1=&RandomRangeFloat(Minimum=-70, Maximum=-100)
     2=&RandomRangeFloat(Minimum=-50, Maximum=-150)
     3=&RandomRangeFloat(Minimum=-70, Maximum=-100)
     4=&RandomRangeFloat(Minimum=-0.5, Maximum=0.5)
     5=&RandomRangeFloat(Minimum=-70, Maximum=-100)
     Index=&SpawnGroup()
     Num=::NewEnumerator4
    ))
   &ScaleColor(
    Scale Alpha=&MergeHelper_FloatSwitch(
     0=&FloatFromCurve(
      FloatCurve=DataInterface() {propertyValues {Curve {keys [
          {arriveTangent=-0.2723, interpMode=RCIM_Cubic, leaveTangent=14.5791, tangentMode=RCTM_Break, time=0, value=0}
          {interpMode=RCIM_Cubic, tangentMode=RCTM_Break, time=0.05, value=1}
          {arriveTangent=0.0378, interpMode=RCIM_Cubic, leaveTangent=0.0378, tangentMode=RCTM_User, time=0.45, value=1.0282}
          {arriveTangent=-3.7547, interpMode=RCIM_Cubic, leaveTangent=-3.7547, tangentMode=RCTM_Break, time=0.7, value=0}
         ]}}}
      Scale Curve=&FNTimeOfDayScalar(Night Scale=0.3, Morning Scale=0.7, Day Scale=0.7, Evening Scale=1)
     )
     1=&FNTimeOfDayScalar(Night Scale=5, Morning Scale=1, Evening Scale=1)
     2=&FloatFromCurve(
      FloatCurve=DataInterface() {propertyValues {Curve {keys [
          {arriveTangent=-0.2723, interpMode=RCIM_Cubic, leaveTangent=14.5791, tangentMode=RCTM_Break, time=0, value=0}
          {interpMode=RCIM_Cubic, tangentMode=RCTM_Break, time=0.1, value=1}
          {arriveTangent=0.0378, interpMode=RCIM_Cubic, leaveTangent=0.0378, tangentMode=RCTM_User, time=0.5, value=1.0186}
          {arriveTangent=-3.7547, interpMode=RCIM_Cubic, leaveTangent=-3.7547, tangentMode=RCTM_Break, time=0.8, value=0}
         ]}}}
      Scale Curve=&FNTimeOfDayScalar(Night Scale=1, Morning Scale=0.7, Day Scale=2, Evening Scale=0.5)
     )
     3=&FloatFromCurve(
      FloatCurve=DataInterface() {propertyValues {Curve {keys [
          {arriveTangent=-0.2723, interpMode=RCIM_Cubic, leaveTangent=14.5791, tangentMode=RCTM_Break, time=0, value=0}
          {interpMode=RCIM_Cubic, tangentMode=RCTM_Break, time=0.1, value=1}
          {arriveTangent=0.0378, interpMode=RCIM_Cubic, leaveTangent=0.0378, tangentMode=RCTM_User, time=0.3785, value=1.0282}
          {arriveTangent=-3.7547, interpMode=RCIM_Cubic, leaveTangent=-3.7547, tangentMode=RCTM_Break, time=0.7, value=0}
         ]}}}
      Scale Curve=&FNTimeOfDayScalar(Night Scale=0.3, Morning Scale=0.7, Day Scale=0.7, Evening Scale=1)
     )
     4=&FloatFromCurve(FloatCurve=DataInterface() {propertyValues {Curve {keys [
          {arriveTangent=6.8822, interpMode=RCIM_Cubic, leaveTangent=6.8822, tangentMode=RCTM_User, time=0, value=0}
          {interpMode=RCIM_Cubic, tangentMode=RCTM_User, time=0.2, value=1}
          {arriveTangent=-0.4524, tangentMode=RCTM_User, time=1, value=0.4}
         ]}}})
     5=&FloatFromCurve(
      FloatCurve=DataInterface() {propertyValues {Curve {keys [
          {arriveTangent=-0.2723, interpMode=RCIM_Cubic, leaveTangent=14.5791, tangentMode=RCTM_Break, time=0, value=0}
          {interpMode=RCIM_Cubic, tangentMode=RCTM_Break, time=0.1, value=1}
          {arriveTangent=0.0378, interpMode=RCIM_Cubic, leaveTangent=0.0378, tangentMode=RCTM_User, time=0.2342, value=1.0089}
          {arriveTangent=-3.7547, interpMode=RCIM_Cubic, leaveTangent=-3.7547, tangentMode=RCTM_Break, time=0.6009, value=-0.029}
         ]}}}
      Scale Curve=&FNTimeOfDayScalar(Night Scale=0.3, Morning Scale=0.7, Day Scale=0.7, Evening Scale=1)
     )
     Index=&SpawnGroup()
     Num=::NewEnumerator4
    )
    Scale RGB=&MergeHelper_VectorSwitch(
     0=&VectorFromFloat(Value=&FNTimeOfDayScalar(Night Scale=1.5, Morning Scale=2, Day Scale=1.8, Evening Scale=2))
     1=&VectorFromFloat(Value=&FNTimeOfDayScalar(Night Scale=5, Morning Scale=2, Day Scale=1.8, Evening Scale=2))
     2=<0.5 0.5 0.5>
     3=&VectorFromFloat(Value=&FNTimeOfDayScalar(Night Scale=1.5, Morning Scale=2, Day Scale=1.8, Evening Scale=2))
     4=&VectorFromFloat(Value=&FNTimeOfDayScalar(Night Scale=3, Morning Scale=1, Evening Scale=1))
     5=&VectorFromFloat(Value=&FNTimeOfDayScalar(Night Scale=2, Morning Scale=2, Day Scale=1.5, Evening Scale=2))
     Index=&SpawnGroup()
     Num=::NewEnumerator4
    )
    ScaleA=true
    ScaleRGB=true
   )
   &GravityForce(Gravity=&MergeHelper_VectorSwitch(
     0=&MakeVector(X=0, Y=0, Z=-980)
     1=<0 0 0>
     2=<0 0 0>
     3=&RandomRangeVector(Minimum=<0 0 -600>, Maximum=<0 0 -980>)
     4=<0 0 -500>
     5=&RandomRangeVector(Minimum=<0 0 -500>, Maximum=<0 0 -980>)
     Index=&SpawnGroup()
     Num=::NewEnumerator4
    ))
   &Drag(
    Drag=&MergeHelper_FloatSwitch(
     0=&RandomRangeFloat(Minimum=3, Maximum=10)
     1=&Multiply_Float(
      A=&RandomRangeFloat(Minimum=8, Maximum=10)
      B=&FloatFromCurve(FloatCurve=DataInterface() {propertyValues {Curve {keys [
           {arriveTangent=-0.112, interpMode=RCIM_Cubic, leaveTangent=-0.112, tangentMode=RCTM_User, time=0, value=1}
           {arriveTangent=-2.6101, interpMode=RCIM_Cubic, leaveTangent=-2.6101, tangentMode=RCTM_User, time=1, value=0}
          ]}}})
     )
     2=&RandomRangeFloat(Minimum=0.2, Maximum=5)
     3=&FloatFromCurve(FloatCurve=DataInterface() {propertyValues {Curve {keys [{interpMode=RCIM_Cubic, tangentMode=RCTM_User, time=0, value=0}, {interpMode=RCIM_Cubic, tangentMode=RCTM_User, time=1, value=1}]}}})
     4=&FloatFromCurve(FloatCurve=DataInterface() {propertyValues {Curve {keys [{interpMode=RCIM_Cubic, tangentMode=RCTM_User, time=0, value=0}, {interpMode=RCIM_Cubic, tangentMode=RCTM_User, time=1, value=1}]}}})
     5=&RandomRangeFloat(Minimum=0.2, Maximum=0.3)
     Index=&SpawnGroup()
     Num=::NewEnumerator4
    )
    Rotational Drag:NiagaraFloat=&MergeHelper_FloatSwitch(0=1, 1=1, 2=1, 3=1, 4=1, Index=&SpawnGroup(), Num=::NewEnumerator3)
    Use Linear Drag=true
    Use Rotational Drag=true
   )
   &AccelerationForce(Acceleration=&MergeHelper_VectorSwitch(
     0=&UniformRangedVector(Minimum=<0 0 -50>, Maximum=<0 0 -500>)
     1=<0 0 0>
     2=<0 0 0>
     3=<0 0 0>
     4=<0 0 0>
     5=<1 1 1>
     Index=&SpawnGroup()
     Num=::NewEnumerator4
    ))
   &SolveForcesAndVelocity(Clamp Velocity=false, Limit Acceleration=false)
   &ScaleSpriteSizeBySpeed(
    Max Scale Factor=&MergeHelper_Vector2Switch(0=<1 1>, 1=<2 5>, 2=<3 5>, 3=<2 2>, 4=<2 4>, 5=<2 2>, Index=&SpawnGroup(), Num=::NewEnumerator4)
    Min Scale Factor=&MergeHelper_Vector2Switch(0=<1 1>, 1=<1 4>, 2=<2 1>, 3=<1 1>, 4=<1 1>, 5=<1 1>, Index=&SpawnGroup(), Num=::NewEnumerator4)
    Sample Scale Factor By Curve=true
    Scale Factor Curve=DataInterface() {propertyValues {Curve {defaultValue=0.5941, keys [{tangentMode=RCTM_User, time=0, value=1}, {tangentMode=RCTM_User, time=1, value=0}]}}}
    Velocity Threshold:NiagaraFloat=&MergeHelper_FloatSwitch(0=250, 1=250, 2=1000, 3=250, 4=250, 5=250, Index=&SpawnGroup(), Num=::NewEnumerator4)
   )
   /SetVariables(
    Particles.0 - Water Emerge DynParam:Vector4f=&MakeVector4(
     X=&FloatFromCurve(FloatCurve=DataInterface() {propertyValues {Curve {keys [
          {arriveTangent=-2.7755, interpMode=RCIM_Cubic, leaveTangent=-2.7755, tangentMode=RCTM_User, time=0.25, value=1}
          {arriveTangent=-0.7689, interpMode=RCIM_Cubic, leaveTangent=-0.7689, tangentMode=RCTM_User, time=1.0011, value=0}
         ]}}})
     Y=&FloatFromCurve(FloatCurve=DataInterface() {propertyValues {Curve {keys [<at -0.0086 -1.3001>,<at 0.9945 1.9153>]}}})
     Z=0
     W=0
    )
    Particles.1 - water crown DynParam:Vector4f=&MakeVector4(
     X=&RandomRangeFloat(
      Minimum=&FloatFromCurve(FloatCurve=DataInterface() {propertyValues {Curve {keys [
           {interpMode=RCIM_Cubic, tangentMode=RCTM_User, time=0, value=1}
           {arriveTangent=-3.7889, interpMode=RCIM_Cubic, leaveTangent=-3.7889, tangentMode=RCTM_User, time=0.75, value=0}
          ]}}})
      Maximum=&FloatFromCurve(FloatCurve=DataInterface() {propertyValues {Curve {keys [
           {interpMode=RCIM_Cubic, tangentMode=RCTM_User, time=0, value=1}
           {arriveTangent=-3.7889, interpMode=RCIM_Cubic, leaveTangent=-3.7889, tangentMode=RCTM_User, time=1, value=0}
          ]}}})
     )
     Y=0
     Z=0
     W=0
    )
    Particles.2 - Mist DynParam:Vector4f=&MakeVector4(
     X=&FloatFromCurve(FloatCurve=DataInterface() {propertyValues {Curve {keys [{arriveTangent=-0.8567, interpMode=RCIM_Cubic, leaveTangent=-0.8567, tangentMode=RCTM_User, time=0, value=0}, <at 1 -0.4 RCIM_Cubic>]}}})
     Y=&RandomRangeFloat(Minimum=0, Maximum=10)
     Z=0
     W=&RandomRangeFloat(Minimum=0.1, Maximum=0.4)
    )
    Particles.3 - Splash Ring DynParam:Vector4f=&MakeVector4(
     X=&FloatFromCurve(FloatCurve=DataInterface() {propertyValues {Curve {
         defaultValue=1
         keys [{interpMode=RCIM_Cubic, tangentMode=RCTM_User, time=0, value=1}, {interpMode=RCIM_Cubic, tangentMode=RCTM_User, time=1, value=0}]
        }}})
     Y=0
     Z=0
     W=0
    )
    Particles.4 - Suplemental splash DynParam:Vector4f=&MakeVector4(
     X=&FloatFromCurve(FloatCurve=DataInterface() {propertyValues {Curve {keys [
          {interpMode=RCIM_Cubic, tangentMode=RCTM_User, time=0, value=1}
          {arriveTangent=-3.7889, interpMode=RCIM_Cubic, leaveTangent=-3.7889, tangentMode=RCTM_User, time=0.8, value=0}
         ]}}})
     Y=0
     Z=0
     W=0
    )
    Particles.5 - end splash DynParam:Vector4f=&MakeVector4(
     X=&FloatFromCurve(FloatCurve=DataInterface() {propertyValues {Curve {keys [{interpMode=RCIM_Cubic, tangentMode=RCTM_Break, time=0, value=1}, {interpMode=RCIM_Cubic, tangentMode=RCTM_Break, time=0.5, value=0.4}]}}})
     Y=0
     Z=0
     W=0
    )
   ) {isSetParameters=true}
   /SetVariables(Particles.1 - stretch water crown velocity:Vector3f=&Lerp_Vector(
     A=&NormalizeVector(
      Vector To Normalize=&Add_Vector(A=$Particles.Initial.ShapeLocation.ShapeNormal, B=<0 0 3>)
      Safe Normalize=true
      Safe Fallback Threshold=0
      Safe Fallback Vector=<0 0 0.001>
     )
     B=&NormalizeVector(
      Vector To Normalize=&Add_Vector(A=$Particles.Initial.ShapeLocation.ShapeNormal, B=<0 0 1>)
      Safe Normalize=true
      Safe Fallback Threshold=0
      Safe Fallback Vector=<0 0 0.001>
     )
     Alpha=&FloatFromCurve(FloatCurve=DataInterface() {propertyValues {Curve {keys [
          {interpMode=RCIM_Cubic, tangentMode=RCTM_User, time=0, value=0}
          {arriveTangent=2.1818, interpMode=RCIM_Cubic, leaveTangent=2.1818, tangentMode=RCTM_User, time=1, value=1}
         ]}}})
    )) {isSetParameters=true}
  ]
  rendererClasses [&NiagaraSpriteRendererProperties]
  renderers [
   {rendererClass=&NiagaraSpriteRendererProperties}
   {rendererClass=&NiagaraSpriteRendererProperties, rendererIndex=1}
   {rendererClass=&NiagaraSpriteRendererProperties, rendererIndex=2}
   {rendererClass=&NiagaraSpriteRendererProperties, rendererIndex=3}
   {rendererClass=&NiagaraSpriteRendererProperties, rendererIndex=4}
  ]
 }
}
summary {
 systemName=NS_Emote_QuicheLorraineLime_Splash
 emitters [
  {emitterName=Droplets001, rendererClasses []}
  {emitterName=Drip_Drop, rendererClasses [&NiagaraSpriteRendererProperties]}
  {emitterName=Edge_Spray_Mesh002, rendererClasses [&NiagaraMeshRendererProperties]}
  {emitterName=Splashes_compressed, rendererClasses [&NiagaraSpriteRendererProperties]}
  {emitterName=LWE_RIPPLE, rendererClasses []}
 ]
 userVariables [
  User.AnimCurve_FXdelay {type=NiagaraFloat, defaultValue=0, description=""}
  User.bIsFrontEnd {type=NiagaraBool}
  User.bIsFrontEndPreview {type=NiagaraBool}
  User.OwnerLocalVelocity {type=Vector3f, description="", defaultValue {struct=Vector3f, value=<0 0 0>}}
  User.RibbonIDOffset {type=NiagaraInt32}
 ]
}
