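# Niagara system for the QuicheLorraineLime traversal emote's water effects: a liquid wake, splashes and boat ripples.
# An anim-trail ribbon and ripple meshes follow a skeletal mesh socket or bone, with trail backward velocity lerped by a system-smoothed value.
# Three sprite emitters throw spray through left/right spawn-group masks, while timed mesh-and-sprite bursts feed lagged position offsets into material parameters.
# Time-of-day inputs adjust color, brightness or opacity across the emitters.
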NiagaraSystem /BRCosmetics/Animation/Game/MainPlayer/Emotes/QuicheLorraineLime/FX/NS_Emote_QuicheLorraineLime_Water_Traversal
dependencies {
 usedDataInterfaces [
  &NiagaraDataInterfaceCurve
  &NiagaraDataInterfaceArrayFloat3
  &NiagaraDataInterfaceVectorCurve
  &NiagaraDataInterfacePlatformSet
  &NiagaraDataInterfaceSpriteRendererInfo
  &NiagaraDataInterfaceSkeletalMesh
 ]
 usedDynamicInputs [
  &RemapRange
  &Add_Float
  &Abs_Float
  &MakeLinearColorFromVectorAndFloat
  &MakeVectorFromLinearColorRGB
 ]
 usedModules [
  &self:SystemSpawnScript.SystemScriptSource.SystemScriptGraph.NiagaraNodeAssignment_1.SetVariables_#1
  &self:SystemSpawnScript.SystemScriptSource.SystemScriptGraph.NiagaraNodeAssignment_3.SetVariables_#2
  &SystemState
  &BlendSpaceHelper
  &self:SystemSpawnScript.SystemScriptSource.SystemScriptGraph.NiagaraNodeAssignment_2.SetVariables_#3
  &self:SystemSpawnScript.SystemScriptSource.SystemScriptGraph.NiagaraNodeAssignment_4.SetVariables_#4
 ]
}
emitters {
 Liquid_Trail {
  emitterUpdateScript [
   &EmitterState(
    Inactive Response=::NewEnumerator1
    Life Cycle Mode=::NewEnumerator1
    Loop Behavior=::NewEnumerator1
    Loop Delay=0.07
    Loop Duration Mode=::NewEnumerator1
    UseLoopDelay=false
   )
   &FN_SpawnAnimTrail(AnimTrailWidthMode=::FromFirst, Spawn Mode:ENiagaraSkeletalSpawning=::NewEnumerator0)
  ]
  particleSpawnScript [
   &FN_InitializeAnimTrailBoneIndecies()
   &FN_InitializeAnimTrail(AnimTrailWidth=90)
   &SkeletalMeshLocation(
    Bone / Socket Index=2
    MaskSpawnGroup=false
    Mesh Sampling Type=::NewEnumerator5
    Orientation Sampling=::NewEnumerator1
    Sampled Position Offset=<0 30 -5>
    Sampled Scale Coordinate Space:ENiagaraCoordinateSpace=::Simulation
    Skeletal Mesh=$System.Skeletal Mesh
    Socket Sampling Mode=::NewEnumerator5
   )
   &Color(Color=&FNTimeOfDayColor(
     Night Color=&RandomRangeLinearColor(Minimum=<rgba 0.1881 0.2601 0.5698 1>, Maximum=<rgba 0.1335 0.2918 0.5698 1>)
     Morning Color=&RandomRangeLinearColor(Minimum=<rgba 0.5052 0.7402 1 1>, Maximum=<rgba 0.1335 0.2918 0.5698 1>)
     Day Color=&RandomRangeLinearColor(Minimum=<rgba 0.5052 0.7402 1 1>, Maximum=<rgba 0.6199 0.7308 0.849 0.8>)
     Evening Color=&RandomRangeLinearColor(Minimum=<rgba 0.5052 0.7402 1 0.5>, Maximum=<rgba 0.2252 0.394 0.6906 0.7>)
    ))
   &AddVelocity(Velocity=&MakeVector(
     X=0
     Y=&Lerp_Float(A=-200, B=-250, Alpha=$System.InterpolateOverTime.InterpolateOverTime_Float.Moving Average)
     Z=0
    )) {module=AddVelocity001}
  ]
  particleUpdateScript [
   &ParticleState()
   &SolveForcesAndVelocity(Clamp Velocity=false, Limit Acceleration=false)
   &ScaleColor(
    Scale Alpha=&FloatFromCurve(
     FloatCurve=DataInterface() {propertyValues {Curve {
        defaultValue=1
        keys [
         {interpMode=RCIM_Cubic, tangentMode=RCTM_User, time=0, value=0}
         {arriveTangent=-0.1462, interpMode=RCIM_Cubic, leaveTangent=-0.1462, tangentMode=RCTM_User, time=0.3, value=1}
         {arriveTangent=-5.4272, interpMode=RCIM_Cubic, leaveTangent=-5.4272, tangentMode=RCTM_User, time=1.0011, value=0}
        ]
       }}}
     Scale Curve=&FNTimeOfDayScalar(Night Scale=1.2, Morning Scale=0.5, Evening Scale=1)
    )
    Scale RGB=&VectorFromFloat(Value=&FNTimeOfDayScalar(Night Scale=2.2, Morning Scale=1.3, Day Scale=2, Evening Scale=1.8))
    ScaleA=true
    ScaleRGB=true
   ) {module=ScaleColor001}
   &FN_UpdateAnimTrail(SkeletalMesh=$System.Skeletal Mesh)
   &ScaleRibbonWidth(Ribbon Width Scale=&FloatFromCurve(FloatCurve=DataInterface() {propertyValues {Curve {keys [
         {arriveTangent=0.464, interpMode=RCIM_Cubic, leaveTangent=0.464, tangentMode=RCTM_User, time=0.0031, value=0.8936}
         {arriveTangent=-2.4767, interpMode=RCIM_Cubic, tangentMode=RCTM_Break, time=0.8, value=0.3359}
        ]}}}))
   &DynamicMaterialParameters(Module.Index0Float0Write=true, Module.Index0Float1Write=true, Module.Index0Float2Write=true, Module.Index0Float3Write=true)
  ]
  rendererClasses [&NiagaraRibbonRendererProperties]
  renderers [
   {rendererClass=&NiagaraRibbonRendererProperties}
   {rendererClass=&NiagaraRibbonRendererProperties, rendererIndex=1}
  ]
 }
 Splash_Both {
  emitterUpdateScript [
   &EmitterState()
   &SpawnRate(SpawnRate=15, Use Spawn Probability=false)
   &SpawnRate(SpawnGroup=1, SpawnRate=15, Use Spawn Probability=false) {module=SpawnRate001}
  ]
  particleSpawnScript [
   &InitializeParticle(
    Color=&RandomRangeLinearColor(Minimum=<rgba 0.5052 0.7402 1 0.5>, Maximum=<rgba 0.2252 0.394 0.6906 0.7>)
    Lifetime Max=0.55
    Lifetime Min=0.35
    Lifetime Mode=::NewEnumerator1
    Sprite Rotation Mode=::NewEnumerator2
    Sprite Size Mode=::NewEnumerator1
    Sprite UV Mode=::NewEnumerator3
    Uniform Sprite Size Max=30
    Uniform Sprite Size Min=18
    UsePositionOffset=false
    Write Material Random=true
   )
   &CameraOffset(Camera Offset Amount=3)
   &ShapeLocation(
    Box Size=<10 30 15>
    MaskSpawnGroup=true
    Offset=<15 15 7>
    Offset Mode=::NewEnumerator0
    Override Seed=false
    Rotation Mode=::NewEnumerator2
    Shape Primitive=::NewEnumerator2
    Surface Only=false
    UseOwnerScale=true
    Yaw / Pitch / Roll=<0.1 0 0>
   )
   &ShapeLocation(
    Box Size=<10 30 15>
    MaskSpawnGroup=true
    Offset=<-15 15 7>
    Offset Mode=::NewEnumerator0
    Override Seed=false
    Rotation Mode=::NewEnumerator2
    Shape Primitive=::NewEnumerator2
    Spawn Group Mask=1
    Surface Only=false
    UseOwnerScale=true
    Yaw / Pitch / Roll=<-0.1 0 0>
   ) {module=ShapeLocation001}
   &AddVelocity(Velocity=&MaskVector(
     Mask Source=0
     Mask Parameter=?
     Mask=0
     Masked Vector=&RandomRangeVector(Minimum=<30 -50 50>, Maximum=<50 -100 70>)
     Passthrough Vector=&MaskVector(
      Mask Source=0
      Mask Parameter=?
      Mask=1
      Masked Vector=&RandomRangeVector(Minimum=<-30 -50 50>, Maximum=<-50 -100 70>)
      Passthrough Vector=<1 1 1>
     )
    ))
   &SubUVAnimation(
    End Frame Range Override=3
    Evaluation Type=::NewEnumerator0
    "Sprite Renderer "=DataInterface() {propertyValues {SpriteRenderer=&self:NiagaraEmitter_2.NiagaraSpriteRendererProperties_1}}
    SubUV Animation Mode:ENiagara_SubUVLookupModeV2=::NewEnumerator2
    UseEndFrame=true
    UseStartFrame=true
   )
  ]
  particleUpdateScript [
   &ParticleState()
   &SpriteRotationRate(Rotation Rate=&MaskFloat(
     Mask Source=0
     Mask Parameter=?
     Mask=0
     Masked Float=&RandomRangeFloat(Minimum=-70, Maximum=-100)
     Passthrough Float=&MaskFloat(
      Mask Source=0
      Mask Parameter=?
      Mask=1
      Masked Float=&RandomRangeFloat(Minimum=70, Maximum=100)
      Passthrough Float=1
     )
    ))
   &GravityForce(Gravity=&RandomRangeVector(Minimum=<0 0 -200>, Maximum=<0 0 -400>))
   &Drag(Drag=&RandomRangeFloat(Minimum=0.2, Maximum=0.3), Use Linear Drag=true, Use Rotational Drag=false)
   &SolveForcesAndVelocity(Clamp Velocity=false, Limit Acceleration=false)
   &ScaleSpriteSize(Uniform Curve Sprite Scale=DataInterface() {propertyValues {Curve {keys [
        {arriveTangent=4.1785, interpMode=RCIM_Cubic, leaveTangent=4.1785, tangentMode=RCTM_User, time=0, value=0.15}
        {arriveTangent=1.1179, interpMode=RCIM_Cubic, leaveTangent=1.1179, tangentMode=RCTM_User, time=0.2042, value=0.6228}
        {arriveTangent=0.0224, interpMode=RCIM_Cubic, leaveTangent=0.0224, tangentMode=RCTM_User, time=1, value=1}
       ]}}})
   &ScaleColor(
    Scale Alpha=&FloatFromCurve(
     FloatCurve=DataInterface() {propertyValues {Curve {keys [
         {arriveTangent=-0.2723, interpMode=RCIM_Cubic, leaveTangent=14.5791, tangentMode=RCTM_Break, time=0, value=0}
         {interpMode=RCIM_Cubic, tangentMode=RCTM_Break, time=0.1, value=1}
         {arriveTangent=0.0378, interpMode=RCIM_Cubic, leaveTangent=0.0378, tangentMode=RCTM_User, time=0.8, value=1}
         {arriveTangent=-3.7547, interpMode=RCIM_Cubic, leaveTangent=-3.7547, tangentMode=RCTM_Break, time=1, value=0}
        ]}}}
     Scale Curve=&FNTimeOfDayScalar(Night Scale=0.12, Morning Scale=0.12, Day Scale=0.12, Evening Scale=0.12)
    )
    Scale RGB=&VectorFromFloat(Value=5)
    ScaleA=true
    ScaleRGB=true
   )
   &DynamicMaterialParameters(
    Index 0 Param 1=&FloatFromCurve(FloatCurve=DataInterface() {propertyValues {Curve {keys [{interpMode=RCIM_Cubic, tangentMode=RCTM_Break, time=0.6, value=1}, {interpMode=RCIM_Cubic, tangentMode=RCTM_Break, time=1, value=0.4}]}}})
    Module.Index0Float0Write=true
    Module.Index0Float1Write=false
    Module.Index0Float2Write=false
    Module.Index0Float3Write=false
   )
  ]
  rendererClasses [&NiagaraSpriteRendererProperties]
  renderers [{rendererClass=&NiagaraSpriteRendererProperties}]
 }
 Splash_Both001 {
  emitterUpdateScript [
   &EmitterState()
   &SpawnRate(SpawnRate=15, Use Spawn Probability=false)
   &SpawnRate(SpawnGroup=1, SpawnRate=15, Use Spawn Probability=false) {module=SpawnRate001}
  ]
  particleSpawnScript [
   &InitializeParticle(
    Color=&RandomRangeLinearColor(Minimum=<rgba 0.5052 0.7402 1 0.5>, Maximum=<rgba 0.2252 0.394 0.6906 0.7>)
    Lifetime Max=0.55
    Lifetime Min=0.35
    Lifetime Mode=::NewEnumerator1
    Sprite Rotation Mode=::NewEnumerator2
    Sprite Size Mode=::NewEnumerator1
    Sprite UV Mode=::NewEnumerator3
    Uniform Sprite Size Max=30
    Uniform Sprite Size Min=18
    UsePositionOffset=false
    Write Material Random=true
   )
   &CameraOffset(Camera Offset Amount=3)
   &ShapeLocation(
    Box Size=<10 30 15>
    MaskSpawnGroup=true
    Offset=<15 -15 7>
    Offset Mode=::NewEnumerator0
    Override Seed=false
    Rotation Mode=::NewEnumerator2
    Shape Primitive=::NewEnumerator2
    Surface Only=false
    UseOwnerScale=true
    Yaw / Pitch / Roll=<0.1 0 0>
   )
   &ShapeLocation(
    Box Size=<10 30 15>
    MaskSpawnGroup=true
    Offset=<-15 -15 7>
    Offset Mode=::NewEnumerator0
    Override Seed=false
    Rotation Mode=::NewEnumerator2
    Shape Primitive=::NewEnumerator2
    Spawn Group Mask=1
    Surface Only=false
    UseOwnerScale=true
    Yaw / Pitch / Roll=<-0.1 0 0>
   ) {module=ShapeLocation001}
   &AddVelocity(
    Velocity=&MaskVector(
     Mask Source=0
     Mask Parameter=?
     Mask=0
     Masked Vector=&RandomRangeVector(Minimum=<30 -50 50>, Maximum=<50 -100 70>)
     Passthrough Vector=&MaskVector(
      Mask Source=0
      Mask Parameter=?
      Mask=1
      Masked Vector=&RandomRangeVector(Minimum=<-30 -50 50>, Maximum=<-50 -100 70>)
      Passthrough Vector=<1 1 1>
     )
    )
    Velocity Speed Scale=1.5
   )
   &SubUVAnimation(
    End Frame Range Override=3
    Evaluation Type=::NewEnumerator0
    "Sprite Renderer "=DataInterface() {propertyValues {SpriteRenderer=&self:NiagaraEmitter_3.NiagaraSpriteRendererProperties_1}}
    SubUV Animation Mode:ENiagara_SubUVLookupModeV2=::NewEnumerator2
    UseEndFrame=true
    UseStartFrame=true
   )
  ]
  particleUpdateScript [
   &ParticleState()
   &SpriteRotationRate(Rotation Rate=&MaskFloat(
     Mask Source=0
     Mask Parameter=?
     Mask=0
     Masked Float=&RandomRangeFloat(Minimum=-70, Maximum=-100)
     Passthrough Float=&MaskFloat(
      Mask Source=0
      Mask Parameter=?
      Mask=1
      Masked Float=&RandomRangeFloat(Minimum=70, Maximum=100)
      Passthrough Float=1
     )
    ))
   &GravityForce(Gravity=&RandomRangeVector(Minimum=<0 0 -200>, Maximum=<0 0 -400>))
   &Drag(Drag=&RandomRangeFloat(Minimum=0.2, Maximum=0.3), Use Linear Drag=true, Use Rotational Drag=false)
   &SolveForcesAndVelocity(Clamp Velocity=false, Limit Acceleration=false)
   &ScaleSpriteSize(Uniform Curve Sprite Scale=DataInterface() {propertyValues {Curve {keys [
        {arriveTangent=4.1785, interpMode=RCIM_Cubic, leaveTangent=4.1785, tangentMode=RCTM_User, time=0, value=0.15}
        {arriveTangent=1.1179, interpMode=RCIM_Cubic, leaveTangent=1.1179, tangentMode=RCTM_User, time=0.2042, value=0.6228}
        {arriveTangent=0.0224, interpMode=RCIM_Cubic, leaveTangent=0.0224, tangentMode=RCTM_User, time=1, value=1}
       ]}}})
   &ScaleColor(
    Scale Alpha=&FloatFromCurve(
     FloatCurve=DataInterface() {propertyValues {Curve {keys [
         {arriveTangent=-0.2723, interpMode=RCIM_Cubic, leaveTangent=14.5791, tangentMode=RCTM_Break, time=0, value=0}
         {interpMode=RCIM_Cubic, tangentMode=RCTM_Break, time=0.1, value=1}
         {arriveTangent=0.0378, interpMode=RCIM_Cubic, leaveTangent=0.0378, tangentMode=RCTM_User, time=0.8, value=1}
         {arriveTangent=-3.7547, interpMode=RCIM_Cubic, leaveTangent=-3.7547, tangentMode=RCTM_Break, time=1, value=0}
        ]}}}
     Scale Curve=&FNTimeOfDayScalar(Night Scale=0.12, Morning Scale=0.12, Day Scale=0.12, Evening Scale=0.12)
    )
    Scale RGB=&VectorFromFloat(Value=5)
    ScaleA=true
    ScaleRGB=true
   )
   &DynamicMaterialParameters(
    Index 0 Param 1=&FloatFromCurve(FloatCurve=DataInterface() {propertyValues {Curve {keys [{interpMode=RCIM_Cubic, tangentMode=RCTM_Break, time=0.6, value=1}, {interpMode=RCIM_Cubic, tangentMode=RCTM_Break, time=1, value=0.4}]}}})
    Module.Index0Float0Write=true
    Module.Index0Float1Write=false
    Module.Index0Float2Write=false
    Module.Index0Float3Write=false
   )
  ]
  rendererClasses [&NiagaraSpriteRendererProperties]
  renderers [{rendererClass=&NiagaraSpriteRendererProperties}]
 }
 Splash_Both002 {
  emitterUpdateScript [
   &EmitterState()
   &SpawnRate(SpawnRate=15, Use Spawn Probability=false)
   &SpawnRate(SpawnGroup=1, SpawnRate=15, Use Spawn Probability=false) {module=SpawnRate001}
  ]
  particleSpawnScript [
   &InitializeParticle(
    Color=&RandomRangeLinearColor(Minimum=<rgba 0.5052 0.7402 1 0.5>, Maximum=<rgba 0.2252 0.394 0.6906 0.7>)
    Lifetime Max=0.55
    Lifetime Min=0.35
    Lifetime Mode=::NewEnumerator1
    Sprite Rotation Mode=::NewEnumerator2
    Sprite Size Mode=::NewEnumerator1
    Sprite UV Mode=::NewEnumerator3
    Uniform Sprite Size Max=30
    Uniform Sprite Size Min=18
    UsePositionOffset=false
    Write Material Random=true
   )
   &CameraOffset(Camera Offset Amount=3)
   &ShapeLocation(
    Box Size=<10 30 15>
    MaskSpawnGroup=true
    Offset=<0 -85 7>
    Offset Mode=::NewEnumerator0
    Override Seed=false
    Rotation Mode=::NewEnumerator2
    Shape Primitive=::NewEnumerator2
    Surface Only=false
    UseOwnerScale=true
    Yaw / Pitch / Roll=<0.1 0 0>
   )
   &ShapeLocation(
    Box Size=<10 30 15>
    MaskSpawnGroup=true
    Offset=<0 -85 7>
    Offset Mode=::NewEnumerator0
    Override Seed=false
    Rotation Mode=::NewEnumerator2
    Shape Primitive=::NewEnumerator2
    Spawn Group Mask=1
    Surface Only=false
    UseOwnerScale=true
    Yaw / Pitch / Roll=<-0.1 0 0>
   ) {module=ShapeLocation001}
   &AddVelocity(
    Velocity=&MaskVector(
     Mask Source=0
     Mask Parameter=?
     Mask=0
     Masked Vector=&RandomRangeVector(Minimum=<30 -50 50>, Maximum=<50 -100 70>)
     Passthrough Vector=&MaskVector(
      Mask Source=0
      Mask Parameter=?
      Mask=1
      Masked Vector=&RandomRangeVector(Minimum=<-30 -50 50>, Maximum=<-50 -100 70>)
      Passthrough Vector=<1 1 1>
     )
    )
    Velocity Speed Scale=1.5
   )
   &SubUVAnimation(
    End Frame Range Override=3
    Evaluation Type=::NewEnumerator0
    "Sprite Renderer "=DataInterface() {propertyValues {SpriteRenderer=&self:NiagaraEmitter_4.NiagaraSpriteRendererProperties_1}}
    SubUV Animation Mode:ENiagara_SubUVLookupModeV2=::NewEnumerator2
    UseEndFrame=true
    UseStartFrame=true
   )
  ]
  particleUpdateScript [
   &ParticleState()
   &SpriteRotationRate(Rotation Rate=&MaskFloat(
     Mask Source=0
     Mask Parameter=?
     Mask=0
     Masked Float=&RandomRangeFloat(Minimum=-70, Maximum=-100)
     Passthrough Float=&MaskFloat(
      Mask Source=0
      Mask Parameter=?
      Mask=1
      Masked Float=&RandomRangeFloat(Minimum=70, Maximum=100)
      Passthrough Float=1
     )
    ))
   &GravityForce(Gravity=&RandomRangeVector(Minimum=<0 0 -200>, Maximum=<0 0 -400>))
   &Drag(Drag=&RandomRangeFloat(Minimum=0.2, Maximum=0.3), Use Linear Drag=true, Use Rotational Drag=false)
   &SolveForcesAndVelocity(Clamp Velocity=false, Limit Acceleration=false)
   &ScaleSpriteSize(Uniform Curve Sprite Scale=DataInterface() {propertyValues {Curve {keys [
        {arriveTangent=4.1785, interpMode=RCIM_Cubic, leaveTangent=4.1785, tangentMode=RCTM_User, time=0, value=0.15}
        {arriveTangent=1.1179, interpMode=RCIM_Cubic, leaveTangent=1.1179, tangentMode=RCTM_User, time=0.2042, value=0.6228}
        {arriveTangent=0.0224, interpMode=RCIM_Cubic, leaveTangent=0.0224, tangentMode=RCTM_User, time=1, value=1}
       ]}}})
   &ScaleColor(
    Scale Alpha=&FloatFromCurve(
     FloatCurve=DataInterface() {propertyValues {Curve {keys [
         {arriveTangent=-0.2723, interpMode=RCIM_Cubic, leaveTangent=14.5791, tangentMode=RCTM_Break, time=0, value=0}
         {interpMode=RCIM_Cubic, tangentMode=RCTM_Break, time=0.1, value=1}
         {arriveTangent=0.0378, interpMode=RCIM_Cubic, leaveTangent=0.0378, tangentMode=RCTM_User, time=0.8, value=1}
         {arriveTangent=-3.7547, interpMode=RCIM_Cubic, leaveTangent=-3.7547, tangentMode=RCTM_Break, time=1, value=0}
        ]}}}
     Scale Curve=&FNTimeOfDayScalar(Night Scale=0.12, Morning Scale=0.12, Day Scale=0.12, Evening Scale=0.12)
    )
    Scale RGB=&VectorFromFloat(Value=5)
    ScaleA=true
    ScaleRGB=true
   )
   &DynamicMaterialParameters(
    Index 0 Param 1=&FloatFromCurve(FloatCurve=DataInterface() {propertyValues {Curve {keys [{interpMode=RCIM_Cubic, tangentMode=RCTM_Break, time=0.6, value=1}, {interpMode=RCIM_Cubic, tangentMode=RCTM_Break, time=1, value=0.4}]}}})
    Module.Index0Float0Write=true
    Module.Index0Float1Write=false
    Module.Index0Float2Write=false
    Module.Index0Float3Write=false
   )
  ]
  rendererClasses [&NiagaraSpriteRendererProperties]
  renderers [{rendererClass=&NiagaraSpriteRendererProperties}]
 }
 Water_Plasma_Splash001 {
  emitterUpdateScript [
   &EmitterState(Life Cycle Mode=::NewEnumerator1, Loop Behavior=::NewEnumerator1, Loop Delay=$System.Delay, UseLoopDelay=false)
   &SpawnBurst_Instantaneous(Spawn Count=15, Spawn Time=0.1, Use Spawn Probability=false, UseLoopCountLimit=false)
   &SpawnBurst_Instantaneous(Spawn Count=15, Spawn Group=2, Spawn Time=3.6, Use Spawn Probability=false, UseLoopCountLimit=false) {module=SpawnBurst_Instantaneous001}
   &SpawnBurst_Instantaneous(Spawn Count=15, Spawn Group=1, Spawn Time=2, Use Spawn Probability=false, UseLoopCountLimit=false) {enabled=false, module=SpawnBurst_Instantaneous002}
   &SpawnRate(
    SpawnRate=&FloatFromCurve(
     FloatCurve=DataInterface() {propertyValues {Curve {keys [
         {interpMode=RCIM_Cubic, tangentMode=RCTM_User, time=0, value=1}
         {arriveTangent=5.5923, interpMode=RCIM_Cubic, leaveTangent=5.5923, tangentMode=RCTM_User, time=0.3085, value=0.1659}
         {interpMode=RCIM_Cubic, tangentMode=RCTM_User, time=1, value=0}
        ]}}}
     CurveIndex=$Emitter.NormalizedLoopAge
     Scale Curve=30
    )
    Use Spawn Probability=false
   ) {enabled=false}
  ]
  particleSpawnScript [
   &PartitionParticles()
   &InitializeParticle(
    Color=<rgba 0.6119 0.7613 1 0.5>
    Lifetime=&RandomRangeFloat(Minimum=0.3, Maximum=0.6)
    Mesh Scale=&SelectVectorFromArray(
     Direct Array Index=&SpawnGroup()
     Vector Selection Array=DataInterface() {propertyValues {FloatData [<0.4 0.5 0.4>, <0.4 0.5 0.4>, <0.4 0.5 0.4>]}}
    )
    Mesh Scale Mode=::NewEnumerator4
    Position Mode=::NewEnumerator1
    UsePositionOffset=true
    Write Material Random=true
   )
   &ShapeLocation(
    Cylinder Radius=10
    MaskSpawnGroup=false
    Offset=<0 -20 0>
    Offset Mode=::NewEnumerator0
    Override Seed=false
    Rotation Angle=90
    Rotation Axis=<0 1 0>
    Rotation Mode=::NewEnumerator1
    Shape Primitive=::NewEnumerator1
    Surface Only=false
    UseOwnerScale=true
   )
   &DynamicMaterialParameters(
    Index 0 Param 1=&RandomRangeFloat(Minimum=0, Maximum=1)
    Module.Index0Param0Write=true
    Module.Index0Param1Write=true
    Module.Index0Param2Write=true
    Module.Index0Param3Write=true
   )
   &AddVelocity(
    Velocity=&SelectVectorFromArray(
     Direct Array Index=$Particles.PartitionParticles.Partition
     Vector Selection Array=DataInterface() {propertyValues {FloatData [<1 0 0>, <-1 0 0>]}}
    )
    Velocity Speed Scale=&RandomRangeFloat(Minimum=20, Maximum=30)
   ) {module=AddVelocity002}
  ]
  particleUpdateScript [
   &ParticleState()
   &GravityForce(Gravity=<0 0 -10>)
   &Drag(Drag=2, Use Linear Drag=true, Use Rotational Drag=true)
   &ScaleColor(
    Scale Alpha=&FloatFromCurve(
     FloatCurve=DataInterface() {propertyValues {Curve {keys [
         {arriveTangent=9.8898, interpMode=RCIM_Cubic, leaveTangent=9.8898, tangentMode=RCTM_User, time=0, value=0}
         {interpMode=RCIM_Cubic, tangentMode=RCTM_User, time=0.3, value=1}
         {arriveTangent=-3.6098, tangentMode=RCTM_User, time=1, value=0}
        ]}}}
     Scale Curve=&FNTimeOfDayScalar(Night Scale=0.5, Morning Scale=1, Evening Scale=1)
    )
    Scale RGB=&VectorFromFloat(Value=&FloatFromCurve(
      FloatCurve=DataInterface() {propertyValues {Curve {keys [
          {arriveTangent=9.0725, interpMode=RCIM_Cubic, leaveTangent=9.0725, tangentMode=RCTM_User, time=0, value=0.5}
          {arriveTangent=-0.7694, interpMode=RCIM_Cubic, leaveTangent=-0.7694, tangentMode=RCTM_User, time=0.1408, value=0.9847}
          {arriveTangent=-0.0405, interpMode=RCIM_Cubic, leaveTangent=-0.0405, tangentMode=RCTM_User, time=1, value=0.5}
         ]}}}
      Scale Curve=&FNTimeOfDayScalar(Night Scale=0.5, Morning Scale=1, Evening Scale=1)
     ))
    ScaleA=true
    ScaleRGB=true
   )
   &InterpolateOverTime(
    Rate Of Change=10
    Target Vector Value:Vector3f=&AddVectorToPosition(
     Position=$Particles.Position
     Vector=&RotateVectorByQuaternion(Vector To Rotate=<0 -200 0>, Quaternion=$Particles.SkeletalMeshLocation.SampledRotation)
    )
    Variable Type=::NewEnumerator2
   )
   &DynamicMaterialParameters(
    Index 0 Param 3=&FloatFromCurve(
     FloatCurve=DataInterface() {propertyValues {Curve {keys [{interpMode=RCIM_Cubic, tangentMode=RCTM_User, time=0, value=0.5}, {interpMode=RCIM_Cubic, tangentMode=RCTM_User, time=1, value=1}]}}}
     Scale Curve=&RandomRangeFloat(Minimum=1, Maximum=2)
    )
    Index 0 Param 4=&FloatFromCurve(
     FloatCurve=DataInterface() {propertyValues {Curve {keys [
         {arriveTangent=-1.9474, interpMode=RCIM_Cubic, leaveTangent=-1.9474, tangentMode=RCTM_User, time=0, value=1}
         {interpMode=RCIM_Cubic, tangentMode=RCTM_User, time=1, value=0}
        ]}}}
     Scale Curve=0.8
    )
    Index 2 Param 1=&MakeFloatFromVector(
     Vector=&Subtract_Position(
      A=&AddVectorToPosition(
       Position=$Particles.Position
       Vector=&RotateVectorByQuaternion(Vector To Rotate=<0 -200 0>, Quaternion=$Particles.SkeletalMeshLocation.SampledRotation)
      )
      B=Linked() {linkedVariable {name=Particles.InterpolateOverTime.InterpolateOverTime_V3_.Moving Average, type=Vector3f}}
     )
     Channel=::NewEnumerator0
    )
    Index 2 Param 2=&MakeFloatFromVector(
     Vector=&Subtract_Position(
      A=&AddVectorToPosition(
       Position=$Particles.Position
       Vector=&RotateVectorByQuaternion(Vector To Rotate=<0 -200 0>, Quaternion=$Particles.SkeletalMeshLocation.SampledRotation)
      )
      B=Linked() {linkedVariable {name=Particles.InterpolateOverTime.InterpolateOverTime_V3_.Moving Average, type=Vector3f}}
     )
     Channel=::NewEnumerator1
    )
    Index 2 Param 3=&MakeFloatFromVector(
     Vector=&Subtract_Position(
      A=&AddVectorToPosition(
       Position=$Particles.Position
       Vector=&RotateVectorByQuaternion(Vector To Rotate=<0 -200 0>, Quaternion=$Particles.SkeletalMeshLocation.SampledRotation)
      )
      B=Linked() {linkedVariable {name=Particles.InterpolateOverTime.InterpolateOverTime_V3_.Moving Average, type=Vector3f}}
     )
     Channel=::NewEnumerator2
    )
    Module.Param0WriteEnabled=false
    Module.Param0WriteEnabled2=true
    Module.Param1WriteEnabled=false
    Module.Param1WriteEnabled2=true
    Module.Param2WriteEnabled=true
    Module.Param2WriteEnabled2=true
    Module.Param3WriteEnabled=true
    Module.Param3WriteEnabled2=true
    Write Parameter Index 2=true
   ) {module=DynamicMaterialParameters001}
   &ScaleMeshSize(Scale Factor=&VectorFromCurve(
     VectorCurve=DataInterface() {propertyValues {
       XCurve {keys [
         {arriveTangent=4.3367, interpMode=RCIM_Cubic, leaveTangent=4.3367, tangentMode=RCTM_User, time=0, value=1}
         {arriveTangent=0.25, interpMode=RCIM_Cubic, leaveTangent=0.25, time=0.4, value=1.65}
         <at 1 1.25>
        ]}
       YCurve {keys [
         {arriveTangent=1.6312, interpMode=RCIM_Cubic, leaveTangent=1.6312, tangentMode=RCTM_User, time=0, value=1}
         <at 0.3 1.2 RCIM_Cubic>
         <at 1 1>
        ]}
       ZCurve {keys [
         {arriveTangent=1.2374, interpMode=RCIM_Cubic, leaveTangent=1.2374, tangentMode=RCTM_User, time=0, value=0.9}
         <at 0.3 1.3 RCIM_Cubic>
         <at 1 0.9 RCIM_Cubic>
        ]}
      }}
     Scale Curve=<0.6 0.6 0.6>
    ))
   &SolveForcesAndVelocity(Clamp Velocity=false, Limit Acceleration=false)
  ]
  rendererClasses [&NiagaraMeshRendererProperties, &NiagaraSpriteRendererProperties]
  renderers [
   {rendererClass=&NiagaraMeshRendererProperties, rendererIndex=0}
   {rendererClass=&NiagaraSpriteRendererProperties, rendererIndex=1}
  ]
 }
 boat_ripples {
  emitterUpdateScript [
   &EmitterState(Inactive Response=::NewEnumerator1, Life Cycle Mode=::NewEnumerator1, Loop Duration=1, UseLoopDelay=false)
   &SpawnBurst_Instantaneous(Use Spawn Probability=false, UseLoopCountLimit=false)
   &SpawnRate(SpawnRate=2, Use Spawn Probability=false)
  ]
  particleSpawnScript [
   &InitializeParticle(
    Color=$System.Ripples color
    Lifetime=&RandomRangeFloat(Minimum=1.5, Maximum=2)
    Mesh Scale=<1.4 1 0.1>
    Mesh Scale Mode=::NewEnumerator4
    Position Offset=<0 0 -50>
    UsePositionOffset=false
    Write Material Random=true
   )
   &DynamicMaterialParameters(
    Index 0 Param 1=&RandomRangeFloat(Minimum=0, Maximum=1)
    Index 0 Param 3=1
    Module.Index0Param0Write=true
    Module.Index0Param1Write=true
    Module.Index0Param2Write=true
    Module.Index0Param3Write=false
   )
   &AddVelocity(Velocity=<0 -50 0>)
   &CameraOffset(Camera Offset Amount=-5)
   &InitialMeshOrientation(Mesh Axis To Orient=<0 1 0>, Mesh Orientation Mode=::NewEnumerator2, Use Rotation Vector=false)
  ]
  particleUpdateScript [
   &ParticleState()
   &SkeletalMeshLocation(
    Bone / Socket Index=2
    Bone Sampling Mode=::NewEnumerator4
    MaskSpawnGroup=false
    Sampled Position Offset=&MakeVector(
     X=0
     Y=0
     Z=&Multiply_Float(
      A=&MakeCustomFloatFromBool(
       Bool=&SetBoolByFloatComparison(
        A=&FN_PickFloatDependOnPlatform(
         TargetPlatform=DataInterface() {propertyValues {Platforms {
            qualityLevelMask=-29
            deviceProfileStates [
             {profileName=Windows, qualityLevelMask=0, setQualityLevelMask=2}
             {profileName=Mobile, qualityLevelMask=0, setQualityLevelMask=4294967295}
             {profileName=Mac, qualityLevelMask=0, setQualityLevelMask=2}
             {profileName=Linux, qualityLevelMask=0, setQualityLevelMask=2}
            ]
           }}}
         TargetPlatformActive=1
         TargetPlatformNoActive=0
        )
        B=0
       )
       True Float=0
       False Float=0
      )
      B=1
     )
    )
    Skeletal Mesh=$System.Skeletal Mesh
   ) {enabled=false}
   &SkeletalMeshLocation(
    Bone / Socket Index=2
    Bone Sampling Mode=::NewEnumerator4
    MaskSpawnGroup=false
    Sampled Position Offset=&Lerp_Vector(
     A=<0 0 8>
     B=<-65 0 12>
     Alpha=&FloatFromCurve(FloatCurve=DataInterface() {propertyValues {Curve {defaultValue=1, keys [{tangentMode=RCTM_User, time=0, value=0}, {tangentMode=RCTM_User, time=1, value=1}]}}})
    )
    Skeletal Mesh=$System.Skeletal Mesh
   ) {module=SkeletalMeshLocation001}
   &ScaleColor(
    Scale Alpha=&FloatFromCurve(FloatCurve=DataInterface() {propertyValues {Curve {keys [
         {arriveTangent=4.2411, interpMode=RCIM_Cubic, leaveTangent=4.2411, tangentMode=RCTM_User, time=0, value=0}
         {interpMode=RCIM_Cubic, tangentMode=RCTM_User, time=0.5, value=1}
         {tangentMode=RCTM_User, time=1, value=0}
        ]}}})
    Scale RGB=&VectorFromFloat(Value=&FNTimeOfDayScalar(Night Scale=1.5, Morning Scale=1.3, Day Scale=1.5, Evening Scale=1.2))
    ScaleA=true
    ScaleRGB=true
   )
   &InterpolateOverTime(
    Rate Of Change=8
    Target Vector Value:Vector3f=&AddVectorToPosition(
     Position=$Particles.Position
     Vector=&RotateVectorByQuaternion(Vector To Rotate=<0 -100 0>, Quaternion=$Particles.SkeletalMeshLocation.SampledRotation)
    )
    Variable Type=::NewEnumerator2
   )
   &ScaleMeshSize(Scale Factor=&VectorFromCurve(
     VectorCurve=DataInterface() {propertyValues {
       XCurve {keys [{arriveTangent=1.898, interpMode=RCIM_Cubic, leaveTangent=1.898, tangentMode=RCTM_User, time=0, value=1}]}
       YCurve {keys [{arriveTangent=0.5312, interpMode=RCIM_Cubic, leaveTangent=0.5312, tangentMode=RCTM_User, time=0, value=1}, <at 1 1.2>]}
       ZCurve {keys [{arriveTangent=9.0725, interpMode=RCIM_Cubic, leaveTangent=9.0725, tangentMode=RCTM_User, time=0, value=1}]}
      }}
     Scale Curve=<1 1 1.5>
    )) {module=ScaleMeshSize001}
   &SolveForcesAndVelocity(Clamp Velocity=false, Limit Acceleration=false)
  ]
  rendererClasses [&NiagaraMeshRendererProperties]
  renderers [{rendererClass=&NiagaraMeshRendererProperties, rendererIndex=0}]
 }
}
summary {
 systemName=NS_Emote_QuicheLorraineLime_Water_Traversal
 emitters [
  {
   emitterName=Water_Plasma_Splash001
   rendererClasses [&NiagaraMeshRendererProperties, &NiagaraSpriteRendererProperties]
  }
  {emitterName=boat_ripples, rendererClasses [&NiagaraMeshRendererProperties]}
  {emitterName=Liquid_Trail, rendererClasses [&NiagaraRibbonRendererProperties]}
  {emitterName=Splash_Both, rendererClasses [&NiagaraSpriteRendererProperties]}
  {emitterName=Splash_Both001, rendererClasses [&NiagaraSpriteRendererProperties]}
  {emitterName=Splash_Both002, rendererClasses [&NiagaraSpriteRendererProperties]}
 ]
 userVariables [
  User.AnimCurve_fogIntensity {type=NiagaraFloat, defaultValue=0, description=""}
  User.AnimCurve_wakeOpacity {type=NiagaraFloat, defaultValue=0, description=""}
  User.bIsFrontEnd {type=NiagaraBool}
  User.bIsFrontEndPreview {type=NiagaraBool}
  User.OwnerLocalVelocity {type=Vector3f, description="", defaultValue {struct=Vector3f, value=<0 0 0>}}
  User.RibbonIDOffset {type=NiagaraInt32}
 ]
}
